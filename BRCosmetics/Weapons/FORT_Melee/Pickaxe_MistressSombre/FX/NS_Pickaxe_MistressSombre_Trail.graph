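# Niagara trail effect for the MistressSombre pickaxe, built from six emitters that follow the weapon's skeletal mesh.
# Three anim-trail ribbon emitters (background, center, detail) form the streak, while distance-spawned smoke, curl-noise stars and mesh circles come from a mesh socket or bone.
# A system ReactivityCheck bool switches background, smoke, stars and circles between green and red-orange palettes; center and detail ribbons keep fixed colors.
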NiagaraSystem /BRCosmetics/Weapons/FORT_Melee/Pickaxe_MistressSombre/FX/NS_Pickaxe_MistressSombre_Trail
dependencies {
 usedDataInterfaces [
  &NiagaraDataInterfaceCurve
  &NiagaraDataInterfaceColorCurve
  &NiagaraDataInterfaceSpriteRendererInfo
  &NiagaraDataInterfaceSkeletalMesh
 ]
 usedDynamicInputs [
  &Boolean_OrOperation
  &InvertBool
  &SetBoolByFloatComparison
  &Boolean_AndOperation
  &Boolean_NotOperation
 ]
 usedModules [
  &self:SystemSpawnScript.SystemScriptSource.SystemScriptGraph.NiagaraNodeAssignment_0.SetVariables_#1
  &self:SystemSpawnScript.SystemScriptSource.SystemScriptGraph.NiagaraNodeAssignment_2.SetVariables_#2
  &self:SystemSpawnScript.SystemScriptSource.SystemScriptGraph.NiagaraNodeAssignment_3.SetVariables_#3
  &self:SystemSpawnScript.SystemScriptSource.SystemScriptGraph.NiagaraNodeAssignment_1.SetVariables_#4
  &SystemState
 ]
}
emitters {
 BGTrail {
  emitterSpawnScript [&FN_RibbonIDOffsetUtility()]
  emitterUpdateScript [
   &EmitterState()
   &FN_SpawnAnimTrail()
  ]
  particleSpawnScript [
   &FN_InitializeAnimTrailBoneIndecies()
   &FN_InitializeAnimTrail(AnimTrailWidth=3.2, Color=<rgba 0.7448 0.7448 0.7448 0.7>, Lifetime=0.45, SkeletalMesh=$System.Skeletal Mesh)
   &AddVelocity(Velocity=<0 400 0>) {enabled=false}
  ]
  particleUpdateScript [
   &SolveForcesAndVelocity(Clamp Velocity=false, Limit Acceleration=false)
   &FN_UpdateAnimTrail(Locked Particle RibbonWidth=1, SkeletalMesh=$System.Skeletal Mesh) {module=FN_UpdateAnimTrail001}
   &ScaleColor(
    Scale Mode=::NewEnumerator1
    Scale RGBA=&MakeVector4_FromLinearColor(LinearColor=&MakeCustomLinearColorFromBool(
      Bool=$System.ReactivityCheck
      True Color=&ColorFromCurve(ColorCurve=DataInterface() {propertyValues {
         AlphaCurve {keys [<at 0 1>,<at 0.7202 0>]}
         BlueCurve {keys [<at 0 0.3022>,<at 0.0863 0.02>,<at 0.3587 0.0018>]}
         GreenCurve {keys [<at 0 2>,<at 0.0863 0.1615>,<at 0.3587 0.026>]}
         RedCurve {keys [<at 0 0.7491>,<at 0.0863 0.031>,<at 0.3587 0.006>]}
        }})
      False Color=&ColorFromCurve(ColorCurve=DataInterface() {propertyValues {
         AlphaCurve {keys [<at 0 1>,<at 0.7202 0>]}
         BlueCurve {keys [<at 0 0>,<at 0.0863 0.1878>,<at 0.3587 0.026>]}
         GreenCurve {keys [<at 0 0.8076>,<at 0.0863 0.0116>,<at 0.3587 0.0019>]}
         RedCurve {keys [<at 0 2>,<at 0.0863 0.0382>,<at 0.3587 0.0093>]}
        }})
     ))
    ScaleRGBA=true
   )
   &ScaleRibbonWidth(Ribbon Width Scale=&FloatFromCurve(FloatCurve=DataInterface() {propertyValues {Curve {keys [{interpMode=RCIM_Cubic, leaveTangent=-1.6691, tangentMode=RCTM_Break, time=0, value=1}, <at 1 0.1>]}}}))
  ]
  rendererClasses [&NiagaraRibbonRendererProperties]
  renderers [
   {rendererClass=&NiagaraRibbonRendererProperties}
   {rendererClass=&NiagaraRibbonRendererProperties, rendererIndex=1}
  ]
 }
 CenterTrail {
  emitterSpawnScript [&FN_RibbonIDOffsetUtility()]
  emitterUpdateScript [
   &EmitterState()
   &FN_SpawnAnimTrail()
  ]
  particleSpawnScript [
   &FN_InitializeAnimTrailBoneIndecies()
   &FN_InitializeAnimTrail(AnimTrailWidth=1.4, Color=<rgba 1.35 1.35 1.35 0.5>, Lifetime=0.3, SkeletalMesh=$System.Skeletal Mesh)
   &AddVelocity(Velocity=<0 400 0>) {enabled=false}
  ]
  particleUpdateScript [
   &SolveForcesAndVelocity(Clamp Velocity=false, Limit Acceleration=false)
   &FN_UpdateAnimTrail(Locked Particle RibbonWidth=1, SkeletalMesh=$System.Skeletal Mesh) {module=FN_UpdateAnimTrail001}
   &ScaleColor(
    Scale Alpha=&FloatFromCurve(FloatCurve=DataInterface() {propertyValues {Curve {keys [{interpMode=RCIM_Cubic, leaveTangent=-5.6227, tangentMode=RCTM_Break, time=0, value=1}, <at 0.6 0>]}}})
    ScaleA=true
    ScaleRGB=true
   )
   &ScaleRibbonWidth(Ribbon Width Scale=&FloatFromCurve(FloatCurve=DataInterface() {propertyValues {Curve {keys [
         {interpMode=RCIM_Cubic, leaveTangent=-2.3018, tangentMode=RCTM_Break, time=0, value=1}
         {interpMode=RCIM_Cubic, leaveTangent=0.0625, time=0.5, value=0.3}
        ]}}}))
  ]
  rendererClasses [&NiagaraRibbonRendererProperties]
  renderers [
   {rendererClass=&NiagaraRibbonRendererProperties}
   {rendererClass=&NiagaraRibbonRendererProperties, rendererIndex=1}
   {rendererClass=&NiagaraRibbonRendererProperties, rendererIndex=2}
  ]
 }
 CirclesTrail {
  emitterSpawnScript [/SetVariables() {isSetParameters=true}]
  emitterUpdateScript [
   &EmitterState(Life Cycle Mode=::NewEnumerator1, Loop Behavior=::NewEnumerator1, Loop Delay=0.07, Loop Duration=1, UseLoopDelay=true)
   &SpawnPerUnit(Module.Use Max Movement Threshold=true, Module.Use Movement Tolerance=false, Use Spawn Probability=false)
   &SpawnRate(SpawnRate=100, Use Spawn Probability=false) {enabled=false}
  ]
  particleSpawnScript [
   &InitializeParticle(
    Color Mode=::NewEnumerator0
    Lifetime=0.25
    Mesh Scale Mode=::NewEnumerator3
    Mesh Uniform Scale=0.1
    UsePositionOffset=false
    Write Material Random=true
   )
   &SkeletalMeshLocation(
    Bone / Socket Index=2
    MaskSpawnGroup=false
    Mesh Sampling Type=::NewEnumerator5
    Orientation Sampling=::NewEnumerator1
    Sampled Position Offset=<0 0 5>
    Skeletal Mesh=$System.Skeletal Mesh
    Socket Sampling Mode=::NewEnumerator5
   )
   &InheritVelocity(Inherited Velocity Speed Limit=250, Speed Limit=true, UseSpeedThreshold=false) {enabled=false}
  ]
  particleUpdateScript [
   &SolveForcesAndVelocity(Clamp Velocity=false, Limit Acceleration=false)
   &ParticleState()
   &ScaleMeshSize(Scale Factor=&VectorFromFloat(Value=&FloatFromCurve(FloatCurve=DataInterface() {propertyValues {Curve {keys [
          {interpMode=RCIM_Cubic, leaveTangent=7.4705, tangentMode=RCTM_Break, time=0, value=0.1}
          {arriveTangent=0.2846, tangentMode=RCTM_User, time=0.3, value=1.05}
         ]}}})))
   &ScaleColor(
    Scale Mode=::NewEnumerator1
    Scale RGBA=&MakeVector4_FromLinearColor(LinearColor=&MakeCustomLinearColorFromBool(
      Bool=$System.ReactivityCheck
      True Color=&ColorFromCurve(ColorCurve=DataInterface() {propertyValues {
         AlphaCurve {keys [<at 0 0.3>,<at 0.5812 0.1>,<at 0.9867 0>]}
         BlueCurve {keys [<at 0 0.8291>,<at 0.2672 0.0094>,<at 1 0.1302>]}
         GreenCurve {keys [<at 0 5>,<at 0.2672 0.1458>,<at 1 0.0165>]}
         RedCurve {keys [<at 0 1.8061>,<at 0.2672 0.0348>,<at 1 0.0364>]}
        }})
      False Color=&ColorFromCurve(ColorCurve=DataInterface() {propertyValues {
         AlphaCurve {keys [<at 0 0.3>,<at 0.5812 0.1>,<at 0.9867 0>]}
         BlueCurve {keys [<at 0 0.81>,<at 0.2672 0.6354>,<at 1 0.1302>]}
         GreenCurve {keys [<at 0 3.2454>,<at 0.2672 0.1044>,<at 1 0.0165>]}
         RedCurve {keys [<at 0 5>,<at 0.2672 0.281>,<at 1 0.0364>]}
        }})
     ))
    ScaleRGBA=true
   )
  ]
  rendererClasses [&NiagaraMeshRendererProperties]
  renderers [{rendererClass=&NiagaraMeshRendererProperties, rendererIndex=0}]
 }
 DetailTrail {
  emitterSpawnScript [&FN_RibbonIDOffsetUtility()]
  emitterUpdateScript [
   &EmitterState()
   &FN_SpawnAnimTrail()
  ]
  particleSpawnScript [
   &FN_InitializeAnimTrailBoneIndecies()
   &FN_InitializeAnimTrail(AnimTrailWidth=1.1, Color=<rgba 1 0.8348 0.0833 0.55>, Lifetime=0.3, SkeletalMesh=$System.Skeletal Mesh)
   &AddVelocity(Velocity=<0 400 0>) {enabled=false}
  ]
  particleUpdateScript [
   &SolveForcesAndVelocity(Clamp Velocity=false, Limit Acceleration=false)
   &FN_UpdateAnimTrail(Locked Particle RibbonWidth=1, SkeletalMesh=$System.Skeletal Mesh) {module=FN_UpdateAnimTrail001}
   &ScaleColor(
    Scale Alpha=&FloatFromCurve(FloatCurve=DataInterface() {propertyValues {Curve {keys [{interpMode=RCIM_Cubic, leaveTangent=-5.6227, tangentMode=RCTM_Break, time=0, value=1}, <at 0.6 0>]}}})
    ScaleA=true
    ScaleRGB=true
   )
   &ScaleRibbonWidth(Ribbon Width Scale=&FloatFromCurve(FloatCurve=DataInterface() {propertyValues {Curve {keys [{interpMode=RCIM_Cubic, leaveTangent=-1.6513, tangentMode=RCTM_Break, time=0, value=1}, <at 1 0.2>]}}}))
  ]
  rendererClasses [&NiagaraRibbonRendererProperties]
  renderers [
   {rendererClass=&NiagaraRibbonRendererProperties}
   {rendererClass=&NiagaraRibbonRendererProperties, rendererIndex=1}
   {rendererClass=&NiagaraRibbonRendererProperties, rendererIndex=2}
  ]
 }
 SmokeTrail {
  emitterUpdateScript [
   &EmitterState(Life Cycle Mode=::NewEnumerator1, Loop Behavior=::NewEnumerator1, Loop Delay=0.085, Loop Duration=1, UseLoopDelay=true)
   &SpawnRate(SpawnRate=100, Use Spawn Probability=false) {enabled=false}
   &SpawnPerUnit(Module.Use Max Movement Threshold=true, Module.Use Movement Tolerance=false, Spawn Spacing=17, Use Spawn Probability=false)
  ]
  particleSpawnScript [
   &InitializeParticle(
    Color=<rgba 1 1 1 0.3>
    Lifetime Max=0.8
    Lifetime Min=0.4
    Lifetime Mode=::NewEnumerator1
    Position Offset=<0 0 35>
    Sprite Rotation Mode=::NewEnumerator2
    Sprite Size Mode=::NewEnumerator1
    Uniform Sprite Size Max=55
    Uniform Sprite Size Min=35
    UsePositionOffset=false
    Write Material Random=true
   )
   &SkeletalMeshLocation(
    Bone / Socket Index=2
    MaskSpawnGroup=false
    Mesh Sampling Type=::NewEnumerator5
    Skeletal Mesh=$System.Skeletal Mesh
    Socket Sampling Mode=::NewEnumerator5
   )
   &ShapeLocation(MaskSpawnGroup=false, Override Seed=false, Sphere Radius=10, UseOwnerScale=true)
   &InheritVelocity(Inherited Velocity Speed Limit=65, Speed Limit=true, UseSpeedThreshold=false)
  ]
  particleUpdateScript [
   &SolveForcesAndVelocity(Clamp Velocity=false, Limit Acceleration=false)
   &ParticleState()
   &ScaleSpriteSize(Uniform Curve Sprite Scale=DataInterface() {propertyValues {Curve {keys [{interpMode=RCIM_Cubic, leaveTangent=0.6699, tangentMode=RCTM_Break, time=0, value=0.6}, {tangentMode=RCTM_User, time=1, value=1.55}]}}})
   &SubUVAnimation(
    "Sprite Renderer "=DataInterface() {propertyValues {SpriteRenderer=&self:SmokeTrail_0.NiagaraSpriteRendererProperties_0}}
    SubUV Animation Mode:ENiagara_SubUVLookupModeV2=::NewEnumerator1
    UseEndFrame=false
    UseStartFrame=false
   )
   &ScaleColor(
    Scale Mode=::NewEnumerator1
    Scale RGBA=&MakeVector4_FromLinearColor(LinearColor=&MakeCustomLinearColorFromBool(
      Bool=$System.ReactivityCheck
      True Color=&ColorFromCurve(ColorCurve=DataInterface() {propertyValues {
         AlphaCurve {keys [<at 0 0.2>,<at 0.1549 1>,<at 1 0>]}
         BlueCurve {keys [<at 0.0717 0.3535>,<at 0.2173 0.0173>,<at 0.575 0.0008>]}
         GreenCurve {keys [<at 0.0717 1.2>,<at 0.2173 0.1302>,<at 0.575 0.0052>]}
         RedCurve {keys [<at 0.0717 0.5469>,<at 0.2173 0.0436>,<at 0.575 0.0014>]}
        }})
      False Color=&ColorFromCurve(ColorCurve=DataInterface() {propertyValues {
         AlphaCurve {keys [<at 0 0.2>,<at 0.1549 1>,<at 1 0>]}
         BlueCurve {keys [<at 0.0717 0.1193>,<at 0.2173 0.2552>,<at 0.575 0.0052>]}
         GreenCurve {keys [<at 0.0717 0.6566>,<at 0.2173 0.0222>,<at 0.575 0.0023>]}
         RedCurve {keys [<at 0.0717 1.2>,<at 0.2173 0.0744>,<at 0.575 0.0033>]}
        }})
     ))
    ScaleRGBA=true
   )
  ]
  rendererClasses [&NiagaraSpriteRendererProperties]
  renderers [{rendererClass=&NiagaraSpriteRendererProperties}]
 }
 StarTrail {
  emitterUpdateScript [
   &EmitterState()
   &SpawnPerUnit(Module.Use Max Movement Threshold=true, Module.Use Movement Tolerance=false, Spawn Spacing=13, Use Spawn Probability=false)
  ]
  particleSpawnScript [
   &InitializeParticle(
    Color=&MakeCustomLinearColorFromBool(Bool=$System.ReactivityCheck, True Color=<rgba 0.4269 2 0.2464 1>, False Color=<rgba 2 0.8769 0.2243 1>)
    Lifetime Max=0.4
    Lifetime Min=0.25
    Lifetime Mode=::NewEnumerator1
    Sprite Size Mode=::NewEnumerator1
    Uniform Sprite Size Max=6
    Uniform Sprite Size Min=3
    UsePositionOffset=false
    Write Material Random=true
   )
   &SkeletalMeshLocation(
    Bone / Socket Index=2
    MaskSpawnGroup=false
    Mesh Sampling Type=::NewEnumerator5
    Skeletal Mesh=$System.Skeletal Mesh
    Socket Sampling Mode=::NewEnumerator5
   )
   &ShapeLocation(MaskSpawnGroup=false, Override Seed=false, Sphere Radius=8, UseOwnerScale=true)
   &InheritVelocity(Inherited Velocity Amount Scale=<1 1.25 1>, Inherited Velocity Speed Limit=65, Speed Limit=true, UseSpeedThreshold=false)
  ]
  particleUpdateScript [
   &CurlNoiseForce(
    Noise Frequency=5
    Noise Strength=&RandomRangeFloat(Minimum=100, Maximum=250)
    Pan Noise=true
    Pan Noise Field=<1 1 1>
   )
   &SolveForcesAndVelocity(Clamp Velocity=false, Limit Acceleration=false)
   &ParticleState()
   &ScaleSpriteSizeBySpeed(Max Scale Factor=<2 6>, Min Scale Factor=<1 2>) {enabled=false}
   &ScaleSpriteSize(Uniform Curve Sprite Scale=DataInterface() {propertyValues {Curve {keys [
        {arriveTangent=9.0725, interpMode=RCIM_Cubic, leaveTangent=9.0725, tangentMode=RCTM_User, time=0, value=0}
        {arriveTangent=0.9199, interpMode=RCIM_Cubic, leaveTangent=0.9199, tangentMode=RCTM_User, time=0.1408, value=0.9847}
        {arriveTangent=-2.5835, interpMode=RCIM_Cubic, leaveTangent=-2.5835, tangentMode=RCTM_User, time=1, value=0}
       ]}}})
  ]
  rendererClasses [&NiagaraSpriteRendererProperties]
  renderers [{rendererClass=&NiagaraSpriteRendererProperties}]
 }
}
summary {
 systemName=NS_Pickaxe_MistressSombre_Trail
 emitters [
  {emitterName=CenterTrail, rendererClasses [&NiagaraRibbonRendererProperties]}
  {emitterName=BGTrail, rendererClasses [&NiagaraRibbonRendererProperties]}
  {emitterName=DetailTrail, rendererClasses [&NiagaraRibbonRendererProperties]}
  {emitterName=SmokeTrail, rendererClasses [&NiagaraSpriteRendererProperties]}
  {emitterName=StarTrail, rendererClasses [&NiagaraSpriteRendererProperties]}
  {emitterName=CirclesTrail, rendererClasses [&NiagaraMeshRendererProperties]}
 ]
 userVariables [
  User.bisReactivityOff {type=NiagaraBool}
  User.IsReactive {type=NiagaraBool}
  User.ReactivityOn {type=NiagaraFloat, defaultValue=1, description=""}
  User.RibbonIDOffset {type=NiagaraInt32}
 ]
}
