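# Parent material for the Mistress Sombre pickaxe, adding a timed emissive reactivity effect, masked to the orb, to the 2023 character parent shading.
# A Timestamp parameter drives a fade-in/fade-out envelope, with debug and on/off controls.
# Camera-space radial coordinates from normals and view vector, panning warped mask samples and Fresnel build colored stripes.
# Material attributes blend toward black base color plus emissive between the PreFX and PostFX functions.
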Material /BRCosmetics/Weapons/FORT_Melee/Pickaxe_MistressSombre/Materials/M_MistressSombre_Parent
graph {
 properties {
  bUseMaterialAttributes=true
  bUsedWithMeshDeformer=true
  bUsedWithSkeletalMesh=true
  pixelDepthOffsetMode=PDOM_Legacy
  shadingModel=MSM_FromMaterialExpression
  usedShadingModels=Default Lit | Subsurface | Clear Coat | Subsurface Profile | Cloth
 }
 nodes [
  n1 = Time()
  n2 = ScalarParameter() {group=MistressSombre, parameterName=Timestamp}
  n3 = Constant() {r=1.5}
  n4 = Subtract(A=@n1, B=Add(A=@n2, B=Add(A=@n3) {constB=3, desc=ExtraDuration}))
  Abs(None=@n4)
  Constant3Vector() {constant=<rgba 0.0193 0.0469 0.003 1>}
  n5 = MaterialFunctionCall() {materialFunction=&MF_CharacterParent_2023_PreFX}
  MaterialFunctionCall(Material=NamedRerouteUsage(), M=@n5.M, Pre Skinned Local Position=@n5.M, Distance Blend=@n5.M, Fresnel=@n5.M) {materialFunction=&MF_CharacterParent_2023_PostFX}
  n6 = NamedRerouteUsage()
  n7 = TextureSample!(UVs=Multiply(A=@n6, B=Constant2Vector() {g=2, r=1}), Tex, "DDX(UVs)"=NamedRerouteUsage(), "DDY(UVs)"=NamedRerouteUsage(), Apply View MipBias) {
   mipValueMode=TMVM_Derivative
   samplerType=SAMPLERTYPE_LinearColor
   texture=&/T_MistressSombre_mask
  }
  n8 = Multiply(A=Subtract(A=@n7.B) {constB=0.5}, B=ScalarParameter() {defaultValue=0.2, group=MistressSombre, parameterName=WarpingIntensity}) {constB=0.2}
  n9 = TextureSample(UVs=Add(A=Panner(Coordinate=@n6) {speedX=-0.2, speedY=0.35}, B=@n8), "DDX(UVs)"=NamedRerouteUsage(), "DDY(UVs)"=NamedRerouteUsage()) {
   mipValueMode=TMVM_Derivative
   samplerType=SAMPLERTYPE_LinearColor
   texture=&T_MistressSombre_mask
  }
  Multiply(A=@n9.R)
  Multiply(A=@n6)
  Multiply(A=@n7.B)
  NamedRerouteDeclaration(None=@n5.Result) {name=InMaterialAttributes, nodeColor=<rgba 0 0.6719 1 1>}
  NamedRerouteDeclaration(None=@n5.Tech Art Mask) {name=InFXMask, nodeColor=<rgba 1 0 0.4219 1>}
  n10 = Abs(None=Subtract(A=@n1, B=@n2))
  n11 = Divide(A=@n4) {constB=1.5, desc=OutLength}
  n12 = Multiply(A=Saturate(None=Divide(A=@n10, B=@n3) {desc=InLength}), B=OneMinus(None=Saturate(None=@n11)))
  n13 = Saturate(None=@n2)
  NamedRerouteDeclaration(None=LinearInterpolate(A=Multiply(A=@n12, B=@n13), Alpha=ScalarParameter() {group=MistressSombre, parameterName=ReactivityDebug}) {constA=1}) {name=ReactivityMask, nodeColor=<rgba 1 0 0.9141 1>}
  n14 = NamedRerouteUsage()
  n15 = Constant()
  n16 = NamedRerouteUsage()
  n17 = Multiply(
   A=ChannelMaskParameter(None=NamedRerouteUsage()) {group=MistressSombre, maskChannel=Alpha, parameterName=OrbFXMask, defaultValue=<rgba 0 0 0 1>}.R
   B=NamedRerouteUsage()
  )
  n18 = ScalarParameter() {defaultValue=1, parameterName=ReactivityOn}
  NamedRerouteDeclaration(None=BlendMaterialAttributes(
    A=NamedRerouteUsage()
    B=SetMaterialAttributes!(MaterialAttributes=@n14, Base Color=@n15, Emissive Color=@n16) {attributeSetTypes [#1, #2]}
    Alpha=Multiply(A=@n17, B=@n18)
   )) {name=OutMaterialAttributes, nodeColor=<rgba 0 1 0.2969 1>}
  n19 = ComponentMask(None=NamedRerouteUsage()) {r=true}
  n20 = Saturate(None=SmoothStep(Value=OneMinus(None=@n19)) {constMax=0.1})
  n21 = TextureSample(UVs=Add(A=Panner(Coordinate=@n6) {speedX=-0.2, speedY=0.25}, B=@n8), "DDX(UVs)"=NamedRerouteUsage(), "DDY(UVs)"=NamedRerouteUsage()) {
   mipValueMode=TMVM_Derivative
   samplerType=SAMPLERTYPE_LinearColor
   texture=&T_MistressSombre_mask
  }
  n22 = Multiply(A=@n21.G, B=ScalarParameter() {defaultValue=1, parameterName=ReactivityBGStripesOpacity}) {constB=0.5}
  n23 = Saturate(None=Multiply(A=Add(A=@n9.R, B=@n22)) {constB=2})
  n24 = NamedRerouteUsage()
  n25 = Multiply(A=SmoothStep(Value=OneMinus(None=@n24)) {constMin=0.3}) {constB=2}
  NamedRerouteDeclaration(None=Multiply(A=Multiply(A=@n20, B=@n23), B=Saturate(None=@n25))) {name=Stripes, nodeColor=<rgba 0 1 0.7188 1>}
  n26 = VectorParameter() {group=MistressSombre, parameterName=ReactivityBGColor, defaultValue=<rgba 0.6525 3 0 1>}
  n27 = Power(
   Base=NamedRerouteDeclaration(None=Fresnel() {baseReflectFraction=0, exponent=1}) {name=Fresnel, nodeColor=<rgba 1 0 0.6797 1>}
   Exp=ScalarParameter() {defaultValue=0.15, parameterName=OrbFresnelExponent}
  )
  n28 = NamedRerouteUsage()
  NamedRerouteDeclaration(None=LinearInterpolate(
    A=LinearInterpolate(A=@n26.RGB, Alpha=@n27) {constB=0}
    B=VectorParameter() {group=MistressSombre, parameterName=ReactivityDetailColor, defaultValue=<rgba 0.3177 0 0.0928 1>}.RGB
    Alpha=Multiply(A=@n28) {constB=0.8}
   )) {name=ReactivityColor, nodeColor=<rgba 1 0.3984 0 1>}
  n29 = DotProduct(A=PixelNormalWS(), B=CameraVectorWS())
  n30 = Transform(None=Subtract(A=PixelNormalWS(), B=CameraVectorWS())) {transformSourceType=TRANSFORMSOURCE_World, transformType=TRANSFORM_Camera}
  n31 = Constant3Vector() {constant=<rgba 1 1 0 0>}
  n32 = Normalize(VectorInput=NamedRerouteDeclaration(None=Multiply(A=@n30, B=@n31)) {name=test, nodeColor=<rgba 0.1016 0 1 1>})
  n33 = Constant3Vector() {constant=<rgba 0 1 0 0>}
  n34 = ComponentMask(None=CrossProduct(A=@n32, B=@n33)) {b=true}
  n35 = Abs(None=@n34)
  n36 = MaterialFunctionCall(In=@n35, Power=Constant() {r=0.75}) {
   desc="Kind of optional, but smoothes out\r\nsome nonlinear distortion"
   materialFunction=&SCurve
  }
  n37 = Multiply(A=QualitySwitch(Default=@n36.Result, Low=@n35, Medium=@n35), B=Sign(None=@n34))
  n38 = ConstantBiasScale(None=@n37) {bias=-3, scale=-0.25}
  n39 = ConstantBiasScale(None=@n37) {scale=0.25}
  n40 = Saturate(None=Sign(None=DotProduct(A=@n32, B=@n33)))
  NamedRerouteDeclaration(None=AppendVector(A=Power(Base=@n29) {constExponent=4, desc=Radial coordinate distortion}, B=LinearInterpolate(A=@n38, B=@n39, Alpha=@n40) {constB=0}) {desc=Output UVs}) {name=RadialCoords, nodeColor=<rgba 0.7578 0 1 1>}
  n41 = TextureCoordinate() {coordinateIndex=1}
  NamedRerouteDeclaration(None=DDX(Value=@n41)) {name=DDX, nodeColor=<rgba 0 0.8594 1 1>}
  NamedRerouteDeclaration(None=DDY(Value=@n41)) {name=DDY, nodeColor=<rgba 1 0.8203 0 1>}
  NamedRerouteUsage()
 ]
}
