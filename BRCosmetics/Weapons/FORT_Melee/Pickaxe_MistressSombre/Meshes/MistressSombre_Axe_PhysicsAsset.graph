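# Physics asset for the Mistress Sombre pickaxe mesh, a two-handed melee cosmetic weapon.
# It holds a single capsule body on the Melee_TwoHanded_Handle bone, laid along the weapon as a simple collision shape.
# It has no constraints, so the weapon simulates as one rigid piece with no articulated parts.
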PhysicsAsset /BRCosmetics/Weapons/FORT_Melee/Pickaxe_MistressSombre/Meshes/MistressSombre_Axe_PhysicsAsset
bodies {Melee_TwoHanded_Handle [{shapeType=Capsule, length=90.1874, radius=9.424, center=<-0.0051 -0.7604 25.8353>, rotation=<rot -0.0067 -90 -0.4761>}]}
constraints []
